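# Niagara animation trail for the Armored Engineer Female pickaxe, drawing ribbons that follow skeletal mesh bones or sockets.
# Ribbon001 spawns two trails in User.LightColor (orange by default), using SpawnGroup-indexed arrays to offset and push them apart along opposite X directions.
# Trail adds a narrower, shorter-lived pale blue-grey ribbon between bone/socket indices 1 and 2, tapering with age.
# Both emitters use shared FN AnimTrail modules and fade alpha over particle life.
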NiagaraSystem /BRCosmetics/Weapons/FORT_Melee/Pickaxe_Armored_Engineer_Female/FX/NS_Pickaxe_ArmoredEngF_AnimTrail
dependencies {
 usedDataInterfaces [&NiagaraDataInterfaceCurve, &NiagaraDataInterfaceArrayFloat3]
 usedDynamicInputs []
 usedModules [&SystemState]
}
emitters {
 Ribbon001 {
  emitterUpdateScript [
   &EmitterState()
   &FN_SpawnAnimTrail_v2(Number of Anim Trails=2)
  ]
  particleSpawnScript [
   /Game/Effects/Niagara/Modules/AnimTrail/FN_InitializeAnimTrail/FN_InitializeAnimTrailBoneIndecies()
   &FN_InitializeAnimTrailBoneIndecies(AnimTrail Selection=1) {module=FN_InitializeAnimTrailBoneIndecies001}
   /Game/Effects/Niagara/Modules/AnimTrail/FN_InitializeAnimTrail/FN_InitializeAnimTrail_v3/FN_InitializeAnimTrail(
    AnimTrailWidth=3
    Color:LinearColor=&MultiplyLinearColor(
     Linear Color=$User.LightColor
     Scale Factor=&MakeFloat4FromVectorAndFloat("Vector (RGB)"=<1 1 1>, "Float (Alpha)"=1)
    )
    Lifetime=0.45
    Position Offset:Vector3f=&TransformVector(
     Vector=&SelectVectorFromArray(Direct Array Index=$Particles.SpawnGroup, Vector Selection Array=DataInterface() {propertyValues {FloatData [<1 0 0>, <-1 0 0>]}})
     Destination Space=::Simulation
    )
   )
   &AddVelocity(Velocity=&SelectVectorFromArray(Direct Array Index=$Particles.SpawnGroup, Vector Selection Array=DataInterface() {propertyValues {FloatData [<30 0 0>, <-30 0 0>]}}))
   &InheritVelocity(Module.Limit Speed=true) {enabled=false}
  ]
  particleUpdateScript [
   &SolveForcesAndVelocity(Clamp Velocity=false, Limit Acceleration=false)
   /Game/Effects/Niagara/Modules/AnimTrail/FN_UpdateAnimTrail/FN_UpdateAnimTrail_v3/FN_UpdateAnimTrail()
   &ParticleState()
   /SetVariables(Particles.VelLength:NiagaraFloat=&self:Velocity_Length(Velocity:Vector3f=$Engine.Owner.Velocity)) {isSetParameters=true}
   &FN_RibbonMaterial(Color=$User.LightColor)
   &ScaleColor(
    Scale Alpha=&FloatFromCurve(FloatCurve=DataInterface() {propertyValues {Curve {keys [<at 0 1>,<at 1 0>]}}})
    ScaleA=true
    ScaleRGB=true
   )
  ]
  rendererClasses [&NiagaraRibbonRendererProperties]
  renderers [{rendererClass=&NiagaraRibbonRendererProperties}]
 }
 Trail {
  emitterUpdateScript [
   &EmitterState()
   &FN_SpawnAnimTrail_v2()
  ]
  particleSpawnScript [
   &FN_InitializeAnimTrailBoneIndecies(First Bone or Socket Index=1, Second Bone or Socket Index=2)
   /Game/Effects/Niagara/Modules/AnimTrail/FN_InitializeAnimTrail/FN_InitializeAnimTrail(AnimTrailWidth=1.1, Color=<rgba 0.5859 0.7227 0.75 1>, Lifetime=0.3)
  ]
  particleUpdateScript [
   &FN_UpdateAnimTrail_v2()
   &ScaleRibbonWidth(Ribbon Width Scale=&FloatFromCurve(FloatCurve=DataInterface() {propertyValues {Curve {keys [
         {arriveTangent=-1.6295, interpMode=RCIM_Cubic, time=0, value=1}
         {arriveTangent=-0.3106, interpMode=RCIM_Cubic, leaveTangent=-0.3106, tangentMode=RCTM_User, time=1, value=0.75}
        ]}}}))
   &ScaleColor(
    Scale Alpha=&FloatFromCurve(FloatCurve=DataInterface() {propertyValues {Curve {keys [
         {arriveTangent=-5.1299, interpMode=RCIM_Cubic, leaveTangent=-5.1299, tangentMode=RCTM_User, time=0, value=0.9998}
         {arriveTangent=-0.9998, interpMode=RCIM_Cubic, leaveTangent=-0.9998, time=0.2369, value=0.498}
         <at 1 0 RCIM_Cubic>
        ]}}})
    ScaleA=true
    ScaleRGB=false
   )
  ]
  rendererClasses [&NiagaraRibbonRendererProperties]
  renderers [{rendererClass=&NiagaraRibbonRendererProperties}]
 }
}
summary {
 systemName=NS_Pickaxe_ArmoredEngF_AnimTrail
 emitters [
  {emitterName=Trail, rendererClasses [&NiagaraRibbonRendererProperties]}
  {emitterName=Ribbon001, rendererClasses [&NiagaraRibbonRendererProperties]}
 ]
 userVariables [
  User.LightColor {type=LinearColor, description="", defaultValue {struct=LinearColor, value=<rgba 1 0.4097 0 1>}}
  User.RibbonIDOffset {type=NiagaraInt32}
  User.SkeletalMesh {
   type=NiagaraDataInterfaceSkeletalMesh
   defaultValue {
    struct=NiagaraExt_VariableValue_DataInterface
    value {dataInterface=&self:NiagaraDataInterfaceSkeletalMesh_0}
   }
  }
 ]
}
