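# Animation Blueprint for the Armored Engineer Female pickaxe that continuously spins its gear bone.
# Each animation update adds a fixed speed, set on initialization, to a stored rotator's roll and wraps it at 360 degrees.
# The AnimGraph feeds that rotator into a Modify Bone node that replaces dyn_gear's rotation over the input pose.
# The step is not scaled by delta time, so spin speed follows update rate.
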AnimBlueprint /BRCosmetics/Weapons/FORT_Melee/Pickaxe_Armored_Engineer_Female/Meshes/Armored_Engineer_Female_Axe_AnimBP
target_skeleton=&/Armored_Engineer_Female_Axe_Skeleton
graphs [
 AnimGraph {nodes [
   n1 = "Transform (Modify) Bone - Bone: dyn_gear"(
    ComponentPose=Local To Component(LocalPose=Input Pose())
    Alpha=1
    Translation="0,0,0"
    Rotation=Get rotation() -> (rotation:Rotator)
    Scale="1,1,1"
   ) {
    rotationMode=BMM_Replace
    scaleMode=BMM_Additive
    translationMode=BMM_Additive
    alphaBoolBlend {blendOption=Linear}
    boneToModify {boneName=dyn_gear}
   }
   Output Pose(Result=Component To Local(ComponentPose=@n1)) {layerGroup=None}
  ]}
 EventGraph {nodes [
   Event BlueprintUpdateAnimation()
   n1 = float + float(A=BreakRotator(InRot=Get rotation()), B=Get rotation_speed() -> (rotation_speed:double)) {is=title K2Node_CommutativeAssociativeBinaryOperator}
   > n2 = Branch(Condition="float >= float"(A=@n1, B=360) {is=title K2Node_CallFunction})
   > Set rotation(rotation:Rotator=MakeRotator(Roll=float - float(A=@n1, B=360) {is=title K2Node_CallFunction})) -> (Output_Get:Rotator)
   Set rotation(execute=@n2.else, rotation=MakeRotator(Roll=@n1))
   Event BlueprintInitializeAnimation()
   > Set rotation_speed(rotation_speed:double=1.25) -> (Output_Get:double)
   TryGetPawnOwner()
  ]}
]
variables [
 rotation {type=Rotator}
 rotation_speed {type=number}
]
